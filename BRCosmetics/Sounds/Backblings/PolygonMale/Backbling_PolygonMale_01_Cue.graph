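# Sound cue for the PolygonMale back bling cosmetic, playing its first backbling sound wave.
# A fixed short delay node feeds a single wave player, so the sound starts slightly after it is triggered.
# It uses the backpack hint attenuation, the back bling sound class and a reactive cosmetic back bling property sheet, with slightly reduced volume.
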SoundCue /BRCosmetics/Sounds/Backblings/PolygonMale/Backbling_PolygonMale_01_Cue
root=Delay_0
nodes [
 {id=Delay_0, children [WavePlayer_3], properties {delayMax=0.35, delayMin=0.35}}
 {id=WavePlayer_3, properties {soundWaveAssetPtr=&/Backbling_PolygonMale_01}}
]
properties {
 attenuationSettings=/Game/Sounds/Attenuation/Fort_BackPack_Hint_Att
 audioPropertiesSheet=/BRCosmetics/Sounds/Settings/PropertySheets/Reactive/APS_Cosmetic_Reactive_Backbling
 maxDistance=1450
 soundClassObject=/Game/Sounds/FortSoundClassesAndModes/Fort_Backblings
 volumeMultiplier=0.9
}
